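; Music encounter Blueprint for the Chapter 6 Spire boss fight, starting and pausing boss combat music as the listener enters or leaves its volume while activated.
; Start spawns the 2D track once on the boss component owner's root and enables a Hero rank-up mute bus mix, which pause lifts.
; Named and gameplay-tag events set the phase parameter for idle, beam attack and exploding.
; A struct event resizes the volume radius.
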(event EventStartMusicEncounter
  (|CallParentFunction)
  (Audio|Modulation|ActivateControlBusMix (MusicEncounter|GetMusicManager) (Variables|Default|GetCBM%20Hero%20Rank%20Level%20Up%20Mute))
  (Variables|Default|GetMusic%20Audio%20Component
    (:then)
    (:else
      (bind _audio_component (MusicEncounter|SpawnMusic2D (Variables|Default|GetCombat%20Music) (Variables|/Script/Engine.Actor|GetRootComponent (Components|GetOwner (MusicEncounter|FindMusicEncounterComponentByTag "(TagName=\"MusicEncounter.Component.Boss\")")))))
      (Variables|Default|SetMusic%20Audio%20Component _audio_component))))

(event EventPauseMusicEncounter
  (|CallParentFunction)
  (Audio|Modulation|DeactivateControlBusMix (MusicEncounter|GetMusicManager) (Variables|Default|GetCBM%20Hero%20Rank%20Level%20Up%20Mute)))

(event MusicEncounter|Volume|EventListenerEntered
  (|CallParentFunction)
  (if (Variables|Default|GetIs%20Activated)
    (CallFunction|StartMusicEncounter)))

(event MusicEncounter|Volume|EventListenerExited
  (|CallParentFunction)
  (if (Variables|Default|GetIs%20Activated)
    (CallFunction|PauseMusicEncounter)))

(event MusicEncounter|EventExecuteNamedEvent
  (params (Music%20Encounter%20Component :type Music%20Encounter%20Component%20Base%20Object%20Reference) (Event%20Name :type Name :default "None"))
  (switch name Event%20Name
    (:StartBeamAttack
      (bind _music_audio_component (Variables|Default|GetMusic%20Audio%20Component)
        (:then
          (Audio|Parameter|SetIntegerParameter _music_audio_component "MusicEncounter.Boss.Phase" 1 :!class /Script/AudioExtensions.AudioParameterControllerInterface))
        (:else)))
    (:StopBeamAttack
      (bind _music_audio_component_2 (Variables|Default|GetMusic%20Audio%20Component)
        (:then
          (Audio|Parameter|SetIntegerParameter _music_audio_component_2 "MusicEncounter.Boss.Phase" 0 :!class /Script/AudioExtensions.AudioParameterControllerInterface))
        (:else)))
    (:Activate
      (Variables|Default|SetIs%20Activated true)
      (if (MusicEncounter|Volume|IsListenerInVolume)
        (CallFunction|StartMusicEncounter)))
    (:Deactivate
      (Variables|Default|SetIs%20Activated false)
      (if (MusicEncounter|Volume|IsListenerInVolume)
        (CallFunction|PauseMusicEncounter)))))

(event MusicEncounter|EventExecuteGameplayTagEvent
  (params (Music%20Encounter%20Component :type Music%20Encounter%20Component%20Base%20Object%20Reference) (Gameplay%20Tag :type Gameplay%20Tag%20Structure))
  (if (== Gameplay%20Tag "(TagName=\"Gameplay.Status.Spire.Exploding\")")
    (bind _music_audio_component (Variables|Default|GetMusic%20Audio%20Component)
      (:then
        (Audio|Parameter|SetIntegerParameter _music_audio_component "MusicEncounter.Boss.Phase" 2 :!class /Script/AudioExtensions.AudioParameterControllerInterface))
      (:else))))

(event MusicEncounter|EventExecuteStructEvent
  (params (Music%20Encounter%20Component :type Music%20Encounter%20Component%20Base%20Object%20Reference) (Event%20Struct :type Instanced%20Struct%20Structure))
  (bind _value (Utilities|InstancedStruct|GetInstancedStructValue Event%20Struct)
    (:Valid
      (bind _output_get (Variables|MusicEncounter|Volume|SetRadius (Utilities|Struct|BreakMusicEncounterSetDynamicVolumeRadius _value)))
      (MusicEncounter|Volume|SetRadius (MusicEncounter|Volume|GetMusicEncounterAGV) _output_get :!class /MusicEncounter/Actors/AudioGameplayVolumes/BPI_MusicEncounterAGVInterface.BPI_MusicEncounterAGVInterface_C))))
