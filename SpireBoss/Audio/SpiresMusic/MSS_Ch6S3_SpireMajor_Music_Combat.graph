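# MetaSound music source for the Spire Major boss combat encounter, driven by boss phase and pause state.
# Phase changes play beat-quantized one-shots, then crossfade alternating loop players on odd/even phases via ADSR envelopes.
# A weapon-fire envelope follower modulates second-stem gain on Default, or high-pass cutoff on the single-stem Mobile page.
# Pausing fades output; the source finishes after that fade or the last phase's one-shot.
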MetaSoundSource /SpireBoss/Audio/SpiresMusic/MSS_Ch6S3_SpireMajor_Music_Combat
inputs [
 MusicEncounter.bIsPaused {type=Bool}
 MusicEncounter.Boss.Phase {type=Int32}
 MusicEncounter.PauseFadeOutTime {type=Time}
 Tempo {type=Float, access=Constructor}
 Time.ElapsedTime {type=Float}
 Time.GameTimeSeconds {type=Float}
 Time.PlayTime {type=Float}
 Time.TimeSinceLastPlayed {type=Float}
 UE.Source.OnPlay {type=Trigger}
]
interfaces [Fort.Time:1.0, UE.OutputFormat.Stereo:1.0, UE.Source.OneShot:1.0, UE.Source:1.1]
outputs [
 "UE.OutputFormat.Stereo.Audio:0" {type=Audio}
 "UE.OutputFormat.Stereo.Audio:1" {type=Audio}
 UE.Source.OneShot.OnFinished {type=Trigger}
]
pages [
 Default {
  is_build_graph=true
  input_defaults {
   MusicEncounter.Boss.Phase=0
   MusicEncounter.PauseFadeOutTime=3
   MusicEncounter.bIsPaused=null
   Tempo=140
   Time.ElapsedTime=0
   Time.GameTimeSeconds=0
   Time.PlayTime=0
   Time.TimeSinceLastPlayed=500
   UE.Source.OnPlay=null
  }
  nodes [
   n1 = UE.BPMToSeconds(BPM=^Tempo) {also {defaults {Beat Multiplier=2}}}
   n2 = UE.BPMToSeconds(BPM=^Tempo)
   n3 = HarmonixNodes.TriggerToTransport(Play=^UE.Source.OnPlay)
   TriggerOnValueChange_1 = UE.Trigger On Value Change:Int32(Value=^MusicEncounter.Boss.Phase)
   TriggerAnyTriggerAny2_0 = "TriggerAny.Trigger Any (2)"(In 1=^UE.Source.OnPlay, In 0=@TriggerOnValueChange_1.Trigger)
   TriggerOnce_0 = UE.Trigger Once(
    Trigger In=HarmonixNodes.MidiClockSubdivisionTrigger(MIDI Clock=HarmonixNodes.Metronome(Tempo=^Tempo, Transport=@n3.Transport).MIDI Clock).Trigger Out
    Reset=UE.Reroute(Value=@TriggerAnyTriggerAny2_0.Out).Value
   ) {also {defaults {Start Closed="<MetaSoundBuilderResult.SUCCEEDED: 0>"}}}
   n4 = VariableDeferredAccessor.Trigger(Variable=InitVariable.Trigger().Variable)
   TriggerOnce_2 = UE.Trigger Once(Trigger In=UE.Reroute(Value=@TriggerOnce_0.Trigger Out).Value, Reset=@n4.Value)
   n5 = VariableAccessor.WaveAsset:Array(Variable=InitVariable.WaveAsset:Array().Variable)
   n6 = UE.Subtract:Int32(PrimaryOperand=Array.Num:WaveAsset:Array(Array=VariableAccessor.WaveAsset:Array(Variable=@n5.Variable).Value).Num) {also {defaults {AdditionalOperands=1}}}
   n7 = VariableAccessor.Int32(Variable=VariableMutator.Int32(Variable=InitVariable.Int32().Variable, Value=Value.Value:Int32(Set=^UE.Source.OnPlay, Target Value=@n6.Out).Output Value).Variable)
   n8 = VariableAccessor.Int32(Variable=@n7.Variable)
   n9 = VariableAccessor.Int32(Variable=@n8.Variable)
   n10 = TriggerCompare.Clamp:Int32(A=^MusicEncounter.Boss.Phase, Compare=@TriggerOnce_2.Trigger Out, B=@n9.Value)
   TriggerRouteTriggerRouteFloat2_0 = "TriggerRoute.Trigger Route (Float, 2):Float"(Set 1=@n10.False, Set 0=@n10.True) {also {defaults {Value 0=1, Value 1=3.5}}}
   TriggerDelay_1 = UE.Trigger Delay(
    Delay Time=UE.BPMToSeconds(BPM=^Tempo, Beat Multiplier=@TriggerRouteTriggerRouteFloat2_0.Value).Seconds
    In=@TriggerRouteTriggerRouteFloat2_0.On Set
   )
   n11 = TriggerCompare.Clamp:Int32(A=UE.Modulo:Int32(PrimaryOperand=^MusicEncounter.Boss.Phase) {also {defaults {AdditionalOperands=2}}}.Out, Compare=@TriggerDelay_1.Out)
   n12 = TriggerCompare.Clamp:Int32(A=^MusicEncounter.Boss.Phase, Compare=@n11.True, B=@n7.Value) {also {defaults {Type=1}}}
   n13 = UE.Reroute(Value=@n12.True)
   n14 = UE.Reroute(Value=@n13.Value)
   n15 = VariableAccessor.WaveAsset:Array(Variable=InitVariable.WaveAsset:Array().Variable)
   n16 = VariableDeferredAccessor.Trigger(Variable=InitVariable.Trigger().Variable)
   WavePlayer_2 = UE.Wave Player:Stereo(
    Wave Asset=Array.Get:WaveAsset:Array(Index=^MusicEncounter.Boss.Phase, Trigger=@n14.Value, Array=@n15.Value).Element
    Play=@n14.Value
    Stop=@n16.Value
   ) {also {defaults {Loop="<MetaSoundBuilderResult.SUCCEEDED: 0>"}}}
   n17 = TriggerCompare.Clamp:Int32(A=UE.Modulo:Int32(PrimaryOperand=^MusicEncounter.Boss.Phase) {also {defaults {AdditionalOperands=2}}}.Out, Compare=@TriggerDelay_1.Out) {also {defaults {B=0, Type=1}}}
   TriggerAnyTriggerAny2_2 = "TriggerAny.Trigger Any (2)"(In 1=UE.Reroute(Value=@n17.True).Value, In 0=UE.Reroute(Value=@n12.False).Value)
   ADSREnvelopeADSREnvelope_0 = ADSR Envelope.ADSR Envelope:Float(
    Release Time=@n1.Seconds
    Attack Time=@n2.Seconds
    Trigger Attack=UE.Reroute(Value=UE.Reroute(Value=@WavePlayer_2.On Play).Value).Value
    Trigger Release=UE.Reroute(Value=@TriggerAnyTriggerAny2_2.Out).Value
   ) {also {defaults {Sustain Level=1}}}
   n18 = TriggerCompare.Clamp:Int32(A=^MusicEncounter.Boss.Phase, Compare=@n17.True, B=@n8.Value) {also {defaults {Type=1}}}
   TriggerAnyTriggerAny2_3 = "TriggerAny.Trigger Any (2)"(In 0=UE.Reroute(Value=@n11.True).Value, In 1=UE.Reroute(Value=@n18.False).Value)
   n19 = UE.Reroute(Value=@n18.True)
   n20 = UE.Reroute(Value=@n19.Value)
   n21 = VariableDeferredAccessor.Trigger(Variable=InitVariable.Trigger().Variable)
   WavePlayer_0 = UE.Wave Player:Stereo(
    Wave Asset=Array.Get:WaveAsset:Array(Index=^MusicEncounter.Boss.Phase, Trigger=@n20.Value, Array=VariableAccessor.WaveAsset:Array(Variable=@n15.Variable).Value).Element
    Play=@n20.Value
    Stop=@n21.Value
   ) {also {defaults {Loop="<MetaSoundBuilderResult.SUCCEEDED: 0>"}}}
   ADSREnvelopeADSREnvelope_1 = ADSR Envelope.ADSR Envelope:Float(Release Time=@n1.Seconds, Attack Time=@n2.Seconds, Trigger Release=@TriggerAnyTriggerAny2_3.Out, Trigger Attack=@WavePlayer_0.On Play) {also {defaults {Sustain Level=1}}}
   DecibelstoLinearGain_1 = UE.Decibels to Linear Gain:Float(Decibels=MapRange.MapRange:Float(In=Modulation.GetModulatorValue() {also {defaults {Modulator=/Game/Sounds/Modulation/EnvelopeFollowers/FortWeapon/CB_WeaponFireEnvFollower}}}.Out) {also {defaults {Out Range A=-3, Out Range B=0}}}.Out Value)
   n22 = UE.Reroute(Value=@n19.Value)
   n23 = VariableAccessor.WaveAsset:Array(Variable=InitVariable.WaveAsset:Array().Variable)
   WavePlayer_1 = UE.Wave Player:Stereo(
    Wave Asset=Array.Get:WaveAsset:Array(Index=^MusicEncounter.Boss.Phase, Trigger=@n22.Value, Array=VariableAccessor.WaveAsset:Array(Variable=@n23.Variable).Value).Element
    Play=@n22.Value
    Stop=@n21.Value
   ) {also {defaults {Loop="<MetaSoundBuilderResult.SUCCEEDED: 0>"}}}
   AudioMixerAudioMixerStereo2_0 = "AudioMixer.Audio Mixer (Stereo, 2)"(
    Gain 1=@DecibelstoLinearGain_1.Linear Gain
    In 0 L=@WavePlayer_0.Out Left
    In 0 R=@WavePlayer_0.Out Right
    In 1 L=@WavePlayer_1.Out Left
    In 1 R=@WavePlayer_1.Out Right
   )
   n24 = UE.Reroute(Value=@n13.Value)
   WavePlayer_4 = UE.Wave Player:Stereo(
    Wave Asset=Array.Get:WaveAsset:Array(Index=^MusicEncounter.Boss.Phase, Trigger=@n24.Value, Array=@n23.Value).Element
    Play=@n24.Value
    Stop=@n16.Value
   ) {also {defaults {Loop="<MetaSoundBuilderResult.SUCCEEDED: 0>"}}}
   AudioMixerAudioMixerStereo2_1 = "AudioMixer.Audio Mixer (Stereo, 2)"(
    Gain 1=@DecibelstoLinearGain_1.Linear Gain
    In 0 L=@WavePlayer_2.Out Left
    In 0 R=@WavePlayer_2.Out Right
    In 1 L=@WavePlayer_4.Out Left
    In 1 R=@WavePlayer_4.Out Right
   )
   DecibelstoLinearGain_0 = UE.Decibels to Linear Gain:Float() {also {defaults {Decibels=6}}}
   n25 = UE.Reroute(Value=@TriggerOnce_0.Trigger Out)
   TriggerDelay_2 = UE.Trigger Delay(Delay Time=UE.BPMToSeconds(BPM=^Tempo).Seconds, In=@n25.Value)
   TriggerOnce_1 = UE.Trigger Once(Trigger In=@n25.Value, Reset=@TriggerDelay_2.Out)
   n26 = UE.Reroute(Value=UE.Reroute(Value=@TriggerOnce_1.Trigger Out).Value)
   WavePlayer_3 = UE.Wave Player:Stereo(Wave Asset=Array.Get:WaveAsset:Array(Index=^MusicEncounter.Boss.Phase, Trigger=@n26.Value, Array=@n5.Value).Element, Play=@n26.Value)
   AudioMixerAudioMixerStereo3_0 = "AudioMixer.Audio Mixer (Stereo, 3)"(
    Gain 1=@ADSREnvelopeADSREnvelope_0.Out Envelope
    Gain 2=@ADSREnvelopeADSREnvelope_1.Out Envelope
    In 2 L=@AudioMixerAudioMixerStereo2_0.Out L
    In 2 R=@AudioMixerAudioMixerStereo2_0.Out R
    In 1 L=@AudioMixerAudioMixerStereo2_1.Out L
    In 1 R=@AudioMixerAudioMixerStereo2_1.Out R
    Gain 0=@DecibelstoLinearGain_0.Linear Gain
    In 0 L=@WavePlayer_3.Out Left
    In 0 R=@WavePlayer_3.Out Right
   )
   TriggerOnValueChange_0 = UE.Trigger On Value Change:Bool(Value=^MusicEncounter.bIsPaused)
   n27 = TriggerCompare.Clamp:Bool(A=^MusicEncounter.bIsPaused, Compare=@TriggerOnValueChange_0.Trigger) {also {defaults {B="<MetaSoundBuilderResult.SUCCEEDED: 0>"}}}
   n28 = Value.Value:Float(Reset=@n27.False, Set=@n27.True) {also {defaults {Init Value=1}}}
   n29 = UE.InterpTo:Audio(Interp Time=^MusicEncounter.PauseFadeOutTime, Target=@n28.Output Value)
   Multiply_0 = UE.Multiply:Audio by Float(PrimaryOperand=@AudioMixerAudioMixerStereo3_0.Out R, AdditionalOperands=@n29.Value)
   Multiply_1 = UE.Multiply:Audio by Float(PrimaryOperand=@AudioMixerAudioMixerStereo3_0.Out L, AdditionalOperands=@n29.Value)
   TriggerDelay_0 = UE.Trigger Delay(Delay Time=^MusicEncounter.PauseFadeOutTime, Reset=@n28.On Reset, In=@n28.On Set)
   TriggerAnyTriggerAny2_1 = "TriggerAny.Trigger Any (2)"(
    In 0=TriggerCompare.Clamp:Int32(A=^MusicEncounter.Boss.Phase, B=VariableAccessor.Int32(Variable=@n9.Variable).Value, Compare=@WavePlayer_3.On Finished).True
    In 1=@TriggerDelay_0.Out
   )
   VariableMutator.Trigger(Value=@ADSREnvelopeADSREnvelope_1.On Done, Variable=@n21.Variable)
   VariableMutator.Trigger(Value=@TriggerDelay_1.Out, Variable=@n4.Variable)
   VariableMutator.Trigger(Value=@ADSREnvelopeADSREnvelope_0.On Done, Variable=@n16.Variable)
  ]
  links [
   {dst="@UE.OutputFormat.Stereo.Audio:1", src=Multiply_0.Out}
   {dst="@UE.OutputFormat.Stereo.Audio:0", src=Multiply_1.Out}
   {dst="@UE.Source.OneShot.OnFinished", src=TriggerAnyTriggerAny2_1.Out}
  ]
 }
 Mobile {
  input_defaults {MusicEncounter.Boss.Phase=0, MusicEncounter.bIsPaused=null}
  nodes [
   InitVariable.WaveAsset:Array()
   n1 = UE.BPMToSeconds(BPM=^Tempo) {also {defaults {Beat Multiplier=2}}}
   n2 = UE.BPMToSeconds(BPM=^Tempo)
   n3 = HarmonixNodes.TriggerToTransport(Play=^UE.Source.OnPlay)
   TriggerOnValueChange_1 = UE.Trigger On Value Change:Int32(Value=^MusicEncounter.Boss.Phase)
   TriggerAnyTriggerAny2_0 = "TriggerAny.Trigger Any (2)"(In 1=^UE.Source.OnPlay, In 0=@TriggerOnValueChange_1.Trigger)
   TriggerOnce_0 = UE.Trigger Once(
    Trigger In=HarmonixNodes.MidiClockSubdivisionTrigger(MIDI Clock=HarmonixNodes.Metronome(Tempo=^Tempo, Transport=@n3.Transport).MIDI Clock).Trigger Out
    Reset=UE.Reroute(Value=@TriggerAnyTriggerAny2_0.Out).Value
   ) {also {defaults {Start Closed="<MetaSoundBuilderResult.SUCCEEDED: 0>"}}}
   n4 = VariableDeferredAccessor.Trigger(Variable=InitVariable.Trigger().Variable)
   TriggerOnce_2 = UE.Trigger Once(Trigger In=UE.Reroute(Value=@TriggerOnce_0.Trigger Out).Value, Reset=@n4.Value)
   n5 = VariableAccessor.WaveAsset:Array(Variable=InitVariable.WaveAsset:Array().Variable)
   n6 = UE.Subtract:Int32(PrimaryOperand=Array.Num:WaveAsset:Array(Array=VariableAccessor.WaveAsset:Array(Variable=@n5.Variable).Value).Num) {also {defaults {AdditionalOperands=1}}}
   n7 = VariableAccessor.Int32(Variable=VariableMutator.Int32(Variable=InitVariable.Int32().Variable, Value=Value.Value:Int32(Set=^UE.Source.OnPlay, Target Value=@n6.Out).Output Value).Variable)
   n8 = VariableAccessor.Int32(Variable=@n7.Variable)
   n9 = VariableAccessor.Int32(Variable=@n8.Variable)
   n10 = TriggerCompare.Clamp:Int32(A=^MusicEncounter.Boss.Phase, Compare=@TriggerOnce_2.Trigger Out, B=@n9.Value)
   TriggerRouteTriggerRouteFloat2_0 = "TriggerRoute.Trigger Route (Float, 2):Float"(Set 1=@n10.False, Set 0=@n10.True) {also {defaults {Value 0=1, Value 1=3.5}}}
   TriggerDelay_0 = UE.Trigger Delay(
    Delay Time=UE.BPMToSeconds(BPM=^Tempo, Beat Multiplier=@TriggerRouteTriggerRouteFloat2_0.Value).Seconds
    In=@TriggerRouteTriggerRouteFloat2_0.On Set
   )
   n11 = TriggerCompare.Clamp:Int32(A=UE.Modulo:Int32(PrimaryOperand=^MusicEncounter.Boss.Phase) {also {defaults {AdditionalOperands=2}}}.Out, Compare=@TriggerDelay_0.Out) {also {defaults {B=0, Type=1}}}
   n12 = TriggerCompare.Clamp:Int32(A=UE.Modulo:Int32(PrimaryOperand=^MusicEncounter.Boss.Phase) {also {defaults {AdditionalOperands=2}}}.Out, Compare=@TriggerDelay_0.Out)
   n13 = TriggerCompare.Clamp:Int32(A=^MusicEncounter.Boss.Phase, Compare=@n12.True, B=@n7.Value) {also {defaults {Type=1}}}
   TriggerAnyTriggerAny2_2 = "TriggerAny.Trigger Any (2)"(In 1=UE.Reroute(Value=@n11.True).Value, In 0=UE.Reroute(Value=@n13.False).Value)
   n14 = UE.Reroute(Value=UE.Reroute(Value=@n13.True).Value)
   n15 = VariableAccessor.WaveAsset:Array(Variable=InitVariable.WaveAsset:Array().Variable)
   n16 = VariableDeferredAccessor.Trigger(Variable=InitVariable.Trigger().Variable)
   WavePlayer_1 = UE.Wave Player:Stereo(
    Wave Asset=Array.Get:WaveAsset:Array(Index=^MusicEncounter.Boss.Phase, Trigger=@n14.Value, Array=@n15.Value).Element
    Play=@n14.Value
    Stop=@n16.Value
   ) {also {defaults {Loop="<MetaSoundBuilderResult.SUCCEEDED: 0>"}}}
   ADSREnvelopeADSREnvelope_0 = ADSR Envelope.ADSR Envelope:Float(
    Release Time=@n1.Seconds
    Attack Time=@n2.Seconds
    Trigger Release=UE.Reroute(Value=@TriggerAnyTriggerAny2_2.Out).Value
    Trigger Attack=UE.Reroute(Value=UE.Reroute(Value=@WavePlayer_1.On Play).Value).Value
   ) {also {defaults {Sustain Level=1}}}
   n17 = TriggerCompare.Clamp:Int32(A=^MusicEncounter.Boss.Phase, Compare=@n11.True, B=@n8.Value) {also {defaults {Type=1}}}
   TriggerAnyTriggerAny2_3 = "TriggerAny.Trigger Any (2)"(In 0=UE.Reroute(Value=@n12.True).Value, In 1=UE.Reroute(Value=@n17.False).Value)
   n18 = UE.Reroute(Value=@n17.True)
   n19 = VariableDeferredAccessor.Trigger(Variable=InitVariable.Trigger().Variable)
   WavePlayer_0 = UE.Wave Player:Stereo(
    Wave Asset=Array.Get:WaveAsset:Array(Index=^MusicEncounter.Boss.Phase, Trigger=@n18.Value, Array=VariableAccessor.WaveAsset:Array(Variable=@n15.Variable).Value).Element
    Play=@n18.Value
    Stop=@n19.Value
   ) {also {defaults {Loop="<MetaSoundBuilderResult.SUCCEEDED: 0>"}}}
   ADSREnvelopeADSREnvelope_1 = ADSR Envelope.ADSR Envelope:Float(Release Time=@n1.Seconds, Attack Time=@n2.Seconds, Trigger Release=@TriggerAnyTriggerAny2_3.Out, Trigger Attack=@WavePlayer_0.On Play) {also {defaults {Sustain Level=1}}}
   DecibelstoLinearGain_0 = UE.Decibels to Linear Gain:Float() {also {defaults {Decibels=6}}}
   n20 = MapRange.MapRange:Float(In=Modulation.GetModulatorValue() {also {defaults {Modulator=&CB_WeaponFireEnvFollower}}}.Out) {also {defaults {Out Range B=200}}}
   n21 = UE.Reroute(Value=@n20.Out Value)
   OnePoleHighPassFilter_2 = UE.One-Pole High Pass Filter:Audio(Cutoff Frequency=UE.Reroute(Value=@n21.Value).Value, In=@WavePlayer_0.Out Right)
   n22 = UE.Reroute(Value=@n20.Out Value)
   OnePoleHighPassFilter_1 = UE.One-Pole High Pass Filter:Audio(Cutoff Frequency=@n22.Value, In=@WavePlayer_1.Out Left)
   OnePoleHighPassFilter_3 = UE.One-Pole High Pass Filter:Audio(Cutoff Frequency=@n21.Value, In=@WavePlayer_0.Out Left)
   OnePoleHighPassFilter_0 = UE.One-Pole High Pass Filter:Audio(Cutoff Frequency=UE.Reroute(Value=@n22.Value).Value, In=@WavePlayer_1.Out Right)
   n23 = UE.Reroute(Value=@TriggerOnce_0.Trigger Out)
   TriggerDelay_2 = UE.Trigger Delay(Delay Time=UE.BPMToSeconds(BPM=^Tempo).Seconds, In=@n23.Value)
   TriggerOnce_1 = UE.Trigger Once(Trigger In=@n23.Value, Reset=@TriggerDelay_2.Out)
   n24 = UE.Reroute(Value=UE.Reroute(Value=@TriggerOnce_1.Trigger Out).Value)
   WavePlayer_2 = UE.Wave Player:Stereo(Wave Asset=Array.Get:WaveAsset:Array(Index=^MusicEncounter.Boss.Phase, Trigger=@n24.Value, Array=@n5.Value).Element, Play=@n24.Value)
   AudioMixerAudioMixerStereo3_0 = "AudioMixer.Audio Mixer (Stereo, 3)"(
    Gain 1=@ADSREnvelopeADSREnvelope_0.Out Envelope
    Gain 2=@ADSREnvelopeADSREnvelope_1.Out Envelope
    Gain 0=@DecibelstoLinearGain_0.Linear Gain
    In 2 R=UE.Reroute(Value=@OnePoleHighPassFilter_2.Out).Value
    In 1 L=UE.Reroute(Value=@OnePoleHighPassFilter_1.Out).Value
    In 2 L=UE.Reroute(Value=@OnePoleHighPassFilter_3.Out).Value
    In 1 R=UE.Reroute(Value=@OnePoleHighPassFilter_0.Out).Value
    In 0 L=@WavePlayer_2.Out Left
    In 0 R=@WavePlayer_2.Out Right
   )
   TriggerOnValueChange_0 = UE.Trigger On Value Change:Bool(Value=^MusicEncounter.bIsPaused)
   n25 = TriggerCompare.Clamp:Bool(A=^MusicEncounter.bIsPaused, Compare=@TriggerOnValueChange_0.Trigger) {also {defaults {B="<MetaSoundBuilderResult.SUCCEEDED: 0>"}}}
   n26 = Value.Value:Float(Reset=@n25.False, Set=@n25.True) {also {defaults {Init Value=1, Target Value=0}}}
   n27 = UE.InterpTo:Audio(Interp Time=^MusicEncounter.PauseFadeOutTime, Target=@n26.Output Value)
   Multiply_0 = UE.Multiply:Audio by Float(PrimaryOperand=@AudioMixerAudioMixerStereo3_0.Out R, AdditionalOperands=@n27.Value)
   Multiply_1 = UE.Multiply:Audio by Float(PrimaryOperand=@AudioMixerAudioMixerStereo3_0.Out L, AdditionalOperands=@n27.Value)
   TriggerDelay_1 = UE.Trigger Delay(Delay Time=^MusicEncounter.PauseFadeOutTime, Reset=@n26.On Reset, In=@n26.On Set)
   TriggerAnyTriggerAny2_1 = "TriggerAny.Trigger Any (2)"(
    In 0=TriggerCompare.Clamp:Int32(A=^MusicEncounter.Boss.Phase, B=VariableAccessor.Int32(Variable=@n9.Variable).Value, Compare=@WavePlayer_2.On Finished).True
    In 1=@TriggerDelay_1.Out
   )
   VariableMutator.Trigger(Value=@ADSREnvelopeADSREnvelope_1.On Done, Variable=@n19.Variable)
   VariableMutator.Trigger(Value=@TriggerDelay_0.Out, Variable=@n4.Variable)
   VariableMutator.Trigger(Value=@ADSREnvelopeADSREnvelope_0.On Done, Variable=@n16.Variable)
  ]
  links [
   {dst="@UE.OutputFormat.Stereo.Audio:1", src=Multiply_0.Out}
   {dst="@UE.OutputFormat.Stereo.Audio:0", src=Multiply_1.Out}
   {dst="@UE.Source.OneShot.OnFinished", src=TriggerAnyTriggerAny2_1.Out}
  ]
 }
]
properties {
 assetImportData=&self:AssetImportData
 numChannels=2
 outputFormat=Stereo
 sampleRate=48000
 soundClassObject=&Fort_Music_NPC
 volume=0.75
 busSends [{audioBus=&AB_Sidechain_FortAmbient_Vol, sendLevel=0.5}]
}
